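# Niagara system for the CrimsonPeak backpack cosmetic: a flickering rim glow that shifts from red at low health to cyan at full health.
# One emitter bursts four looping sprites, partitioned onto skeletal mesh bones or sockets with per-index offsets.
# Irregular curves flicker sprite size and light intensity, with a light renderer lighting a subset chosen by partition index.
# Alpha differs between gameplay and front end.
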NiagaraSystem /BRCosmetics/Accessories/FORT_Backpacks/Backpack_M_MED_CrimsonPeak/FX/NS_Backpack_CrimsonPeak
dependencies {
 usedDataInterfaces [&NiagaraDataInterfaceArrayFloat3, &NiagaraDataInterfaceCurve]
 usedDynamicInputs [&Vector2DFromFloat]
 usedModules [
  &SystemState
  &self:SystemSpawnScript.SystemScriptSource.SystemScriptGraph.NiagaraNodeAssignment_0.SetVariables_#1
 ]
}
emitters {Rim_Glow {
  emitterUpdateScript [
   &EmitterState()
   &SpawnBurst_Instantaneous(Spawn Count=4, Use Spawn Probability=false, UseLoopCountLimit=true)
  ]
  particleSpawnScript [
   &PartitionParticles(Number Of Partitions=4)
   &InitializeParticle(
    Color=<rgba 1 1 1 0.1>
    Lifetime=2
    Sprite Size Mode=::NewEnumerator3
    Uniform Sprite Size=22
    UsePositionOffset=false
    Write Material Random=true
   )
   &SkeletalMeshLocation(
    Bone / Socket Index=$Particles.PartitionParticles.Partition
    Bone Sampling Mode=::NewEnumerator4
    Interpolate Mesh Sampling=false
    MaskSpawnGroup=false
    Sampled Position Offset=&SelectVectorFromArray(
     Direct Array Index=$Particles.PartitionParticles.Partition
     Vector Selection Array=DataInterface() {propertyValues {FloatData [<0 -3 8>, <0 -4.25 1>, <0 -3 -7>, <0 -1 -12>]}}
    )
    Skeletal Mesh=$User.Skeletal Mesh
   )
   &CameraOffset(Camera Offset Amount=5)
  ]
  particleUpdateScript [
   &ParticleState(
    Kill Particles When Lifetime Has Elapsed=false
    Let Infinitely Lived Particles Die When Emitter Deactivates=true
    Loop Particles Lifetime=true
   )
   &ScaleSpriteSize(Uniform Curve Sprite Scale=DataInterface() {propertyValues {Curve {keys [
        {arriveTangent=2.3436, interpMode=RCIM_Cubic, time=0, value=1}
        {arriveTangent=0.0331, interpMode=RCIM_Cubic, leaveTangent=0.0331, time=0.0802, value=0.9796}
        {arriveTangent=-0.0227, interpMode=RCIM_Cubic, leaveTangent=-0.0227, tangentMode=RCTM_User, time=0.1732, value=1.0057}
        {arriveTangent=-0.1715, interpMode=RCIM_Cubic, leaveTangent=-0.1715, time=0.2228, value=0.968}
        {arriveTangent=0.0866, interpMode=RCIM_Cubic, leaveTangent=0.0866, time=0.2792, value=0.9876}
        {arriveTangent=0.1114, interpMode=RCIM_Cubic, leaveTangent=0.1114, time=0.3503, value=0.979}
        {arriveTangent=-0.095, interpMode=RCIM_Cubic, leaveTangent=-0.095, time=0.411, value=1.0022}
        {arriveTangent=-0.1898, interpMode=RCIM_Cubic, leaveTangent=-0.1898, time=0.456, value=0.969}
        {arriveTangent=0.004, interpMode=RCIM_Cubic, leaveTangent=0.004, time=0.4903, value=0.9872}
        {arriveTangent=0.0372, interpMode=RCIM_Cubic, leaveTangent=0.0372, time=0.5521, value=0.9694}
        {arriveTangent=0.0937, interpMode=RCIM_Cubic, leaveTangent=0.0937, time=0.6136, value=0.9918}
        {arriveTangent=0.1958, interpMode=RCIM_Cubic, leaveTangent=0.1958, time=0.6919, value=0.9825}
        {arriveTangent=-0.0924, interpMode=RCIM_Cubic, leaveTangent=-0.0924, time=0.7551, value=1.0195}
        {arriveTangent=-0.3193, interpMode=RCIM_Cubic, leaveTangent=-0.3193, time=0.8008, value=0.9724}
        {arriveTangent=-0.0587, interpMode=RCIM_Cubic, leaveTangent=-0.0587, time=0.8683, value=0.9834}
        {arriveTangent=0.1264, interpMode=RCIM_Cubic, leaveTangent=0.1264, time=0.9304, value=0.9648}
        {interpMode=RCIM_Cubic, leaveTangent=0.0463, time=1, value=1}
       ]}}})
   &DynamicMaterialParameters(
    Index 0 Param 1=3
    Index 0 Param 2=0.5
    Index 0 Param 4=100
    Module.Param0WriteEnabled=true
    Module.Param1WriteEnabled=true
    Module.Param2WriteEnabled=true
    Module.Param3WriteEnabled=true
   )
   &ScaleColor(
    Scale Alpha=&FN_FrontEndFloatModifier(Gameplay=0.7, Front End=0.5)
    Scale RGB=&MakeVectorFromLinearColorRGB(Color=&LerpLinearColors(StartColor=<rgba 1 0.0312 0.0411 1>, EndColor=<rgba 0.2549 2 2 1>, LerpFactor=$System.CurrentHealth))
    ScaleA=true
    ScaleRGB=true
   )
   &FN_Light_Attributes(
    Color=$Particles.Color
    Create Color Attribute=true
    Create Exponent Attribute=true
    Create Position Attribute=true
    Create Radius Attribute=true
    Enable Particle Light Enable=true
    Exponent=2.5
    Light Enabled:NiagaraBool=&SetBoolByIntComparison(A=$Particles.PartitionParticles.Partition, B=1, Comparison Type=::NewEnumerator7)
    Position=&SelectVectorFromArray(
     Direct Array Index=$Particles.PartitionParticles.Partition
     Vector Selection Array=DataInterface() {propertyValues {FloatData [<-1.5 -10 6>, <-2 -10 -8>, <0 0 0>, <0 0 0>]}}
    )
    Radius=20
    Scale Color=&FloatFromCurve(
     FloatCurve=DataInterface() {propertyValues {Curve {keys [
         <at 0 1 RCIM_Cubic>
         {arriveTangent=0.4594, interpMode=RCIM_Cubic, leaveTangent=0.4594, time=0.0669, value=0.8288}
         {arriveTangent=-0.0227, interpMode=RCIM_Cubic, leaveTangent=-0.0227, tangentMode=RCTM_User, time=0.1294, value=1.0595}
         {arriveTangent=-2.2496, interpMode=RCIM_Cubic, leaveTangent=-2.2496, time=0.1628, value=0.7265}
         {arriveTangent=1.1356, interpMode=RCIM_Cubic, leaveTangent=1.1356, time=0.2007, value=0.8992}
         {arriveTangent=1.4606, interpMode=RCIM_Cubic, leaveTangent=1.4606, time=0.2486, value=0.8239}
         {arriveTangent=-1.2459, interpMode=RCIM_Cubic, leaveTangent=-1.2459, time=0.2893, value=1.0286}
         {arriveTangent=-2.4891, interpMode=RCIM_Cubic, leaveTangent=-2.4891, time=0.3196, value=0.7354}
         {arriveTangent=0.052, interpMode=RCIM_Cubic, leaveTangent=0.052, time=0.3426, value=0.8959}
         {arriveTangent=0.4885, interpMode=RCIM_Cubic, leaveTangent=0.4885, time=0.3842, value=0.7387}
         {arriveTangent=1.2284, interpMode=RCIM_Cubic, leaveTangent=1.2284, time=0.4256, value=0.9365}
         {arriveTangent=2.5684, interpMode=RCIM_Cubic, leaveTangent=2.5684, time=0.4783, value=0.8543}
         {arriveTangent=-1.212, interpMode=RCIM_Cubic, leaveTangent=-1.212, time=0.5208, value=1.1809}
         {arriveTangent=-4.1883, interpMode=RCIM_Cubic, leaveTangent=-4.1883, time=0.5515, value=0.7655}
         {arriveTangent=-0.7695, interpMode=RCIM_Cubic, leaveTangent=-0.7695, time=0.5969, value=0.8621}
         {arriveTangent=-0.5924, interpMode=RCIM_Cubic, leaveTangent=-0.5924, time=0.6386, value=0.6984}
         {arriveTangent=0.1752, interpMode=RCIM_Cubic, leaveTangent=0.1752, time=0.6788, value=0.8136}
         {arriveTangent=1.1613, interpMode=RCIM_Cubic, leaveTangent=1.1613, time=0.7211, value=0.7129}
         {arriveTangent=0.6146, interpMode=RCIM_Cubic, leaveTangent=0.6146, time=0.7516, value=0.8982}
         {arriveTangent=0.5942, interpMode=RCIM_Cubic, leaveTangent=0.5942, time=0.783, value=0.7509}
         {arriveTangent=1.852, interpMode=RCIM_Cubic, leaveTangent=1.852, time=0.8244, value=0.9415}
         {arriveTangent=1.1296, interpMode=RCIM_Cubic, leaveTangent=1.1296, time=0.838, value=0.8527}
         {arriveTangent=1.2933, interpMode=RCIM_Cubic, leaveTangent=1.2933, time=0.9054, value=1.0329}
         {arriveTangent=0.7504, interpMode=RCIM_Cubic, leaveTangent=0.7504, time=0.9341, value=0.9771}
         {arriveTangent=0.3479, interpMode=RCIM_Cubic, leaveTangent=0.3479, time=0.9623, value=1.0756}
         {interpMode=RCIM_Cubic, leaveTangent=0.0463, time=1, value=1}
        ]}}}
     Scale Curve=2
    )
   )
  ]
  rendererClasses [&NiagaraSpriteRendererProperties, &NiagaraLightRendererProperties]
  renderers [
   {rendererClass=&NiagaraSpriteRendererProperties}
   {rendererClass=&NiagaraLightRendererProperties, rendererIndex=1}
  ]
 }}
summary {
 systemName=NS_Backpack_CrimsonPeak
 emitters [{emitterName=Rim_Glow, rendererClasses [&NiagaraSpriteRendererProperties, &NiagaraLightRendererProperties]}]
 userVariables [
  User.bIsFrontEnd {type=NiagaraBool}
  User.bIsFrontEndPreview {type=NiagaraBool}
  User.Health {type=NiagaraFloat, defaultValue=1, description=""}
  User.Skeletal Mesh {
   type=NiagaraDataInterfaceSkeletalMesh
   defaultValue {
    struct=NiagaraExt_VariableValue_DataInterface
    value {dataInterface=&self:NiagaraDataInterfaceSkeletalMesh_0}
   }
  }
 ]
}
