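# Material function that overlays a health gauge on the Crimson Peak backpack, altering base color and emissive of incoming material attributes.
# A thresholded fill, clamped by a Blueprint-fed hold time, blends empty and health colors, then tints toward damage color by Blueprint damage flag and reactivity.
# Named reroutes define a texture channel mask, UV dial cutoff, bounds-normalized pre-skinned position, sine-based pulse and gradient timed from Blueprint game time.
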MaterialFunction /BRCosmetics/Accessories/FORT_Backpacks/Backpack_M_MED_CrimsonPeak/Material/MF_CrimsonPeak_HealthDisplay
graph {nodes [
  n1 = GetMaterialAttributes(None=FunctionInput() {inputType=FunctionInput_MaterialAttributes}) {attributeGetTypes [#1, #2]}
  FunctionOutput(None=SetMaterialAttributes!(MaterialAttributes=@n1.MaterialAttributes, BaseColor=NamedRerouteUsage(), EmissiveColor=NamedRerouteUsage()) {attributeSetTypes [#1, #2]})
  n2 = VectorParameter() {group=CrimsonPeak, parameterName=Health_Color, defaultValue=<rgba 0.005 0.25 0.25 1>}
  n3 = OneMinus(None=NamedRerouteUsage())
  n4 = Multiply(A=NamedRerouteUsage())
  n5 = ScalarParameter() {defaultValue=0.34, group=CrimsonPeak, parameterName=HoldTimeFromBP}
  n6 = Saturate(None=Add(A=Subtract(A=@n3), B=Clamp(None=@n4, Max=@n5)))
  n7 = Multiply(A=Multiply(A=Ceil(None=@n6), B=NamedRerouteUsage()), B=NamedRerouteUsage())
  Multiply(A=@n2.RGB, B=@n7)
  n8 = NamedRerouteUsage()
  NamedRerouteDeclaration(None=Multiply(A=OneMinus(None=@n8), B=NamedRerouteUsage())) {name=maskAlpha, nodeColor=<rgba 0.5 0 1 1>}
  n9 = TextureSampleParameter2D() {
   group=CrimsonPeak
   parameterName=Mask_Texture
   samplerType=SAMPLERTYPE_Masks
   texture=/BRCosmetics/Accessories/FORT_Backpacks/Backpack_M_MED_CrimsonPeak/Textures/T_M_MED_CrimsonPeak_Backpack_FX
  }
  NamedRerouteDeclaration(None=StaticComponentMaskParameter(None=@n9.RGB) {defaultB=true, group=CrimsonPeak, parameterName=Mask_Channel}) {name=healthMask, nodeColor=<rgba 1 0.0234 0 1>}
  n10 = Sine(None=Multiply(A=ScalarParameter() {defaultValue=0.8, group=CrimsonPeak, parameterName=Pulse Speed}, B=Time()))
  n11 = ScalarParameter() {defaultValue=0.7, group=CrimsonPeak, parameterName=Pulse Strength}
  NamedRerouteDeclaration(None=Clamp(None=Frac(None=@n10), Min=OneMinus(None=@n11)) {minDefault=0.25}) {name=Pulse, nodeColor=<rgba 0.7344 0 1 1>}
  n12 = LinearInterpolate(A=VectorParameter() {group=CrimsonPeak, parameterName=Empty_Color, defaultValue=<rgba 0.0003 0.026 0.0228 1>}.RGB, B=@n2.RGB, Alpha=@n7)
  n13 = LinearInterpolate(
   A=@n12
   B=LinearInterpolate(
    A=@n12
    B=VectorParameter() {parameterName=DamageColor, defaultValue=<rgba 1 0 0 1>}.RGB
    Alpha=ScalarParameter() {group=CrimsonPeak, parameterName=IsDamagedFromBP}
   )
   Alpha=ScalarParameter() {defaultValue=1, group=CrimsonPeak, parameterName=CrimsonPeak_Reactivity}
  )
  NamedRerouteDeclaration(None=@n13) {name=outEmissive, nodeColor=<rgba 1 0.2812 0 1>}
  NamedRerouteDeclaration(None=@n1.BaseColor) {name=inBaseColor, nodeColor=<rgba 0.8047 1 0 1>}
  n14 = Time()
  NamedRerouteDeclaration(None=Frac(None=Multiply(A=@n14) {constB=0.25})) {name=LocalDebug, nodeColor=<rgba 0 0.1797 1 1>}
  NamedRerouteDeclaration(None=@n1.EmissiveColor) {name=inEmissiveColor, nodeColor=<rgba 0.1719 1 0 1>}
  n15 = NamedRerouteUsage()
  n16 = ScalarParameter() {defaultValue=0.65, group=CrimsonPeak, parameterName=Color_Opacity}
  NamedRerouteDeclaration(None=LinearInterpolate(A=NamedRerouteUsage(), B=NamedRerouteUsage(), Alpha=Multiply(A=@n15, B=@n16) {constB=0.25})) {name=outBaseColor, nodeColor=<rgba 0 1 0.1562 1>}
  n17 = NamedRerouteUsage()
  n18 = ScalarParameter() {defaultValue=0.75, group=CrimsonPeak, parameterName=Emissive_Opacity}
  NamedRerouteDeclaration(None=LinearInterpolate(A=NamedRerouteUsage(), B=NamedRerouteUsage(), Alpha=Multiply(A=@n17, B=@n18))) {name=outEmissiveColor, nodeColor=<rgba 0.0781 1 0 1>}
  n19 = ComponentMask(None=TextureCoordinate()) {r=true}
  n20 = ScalarParameter() {defaultValue=0.185, group=CrimsonPeak, parameterName=DialMask_Pos}
  NamedRerouteDeclaration(None=Ceil(None=Subtract(A=@n19, B=@n20))) {name=gaugeMask, nodeColor=<rgba 0 1 0.6016 1>}
  NamedRerouteDeclaration(None=@n13) {name=outColor, nodeColor=<rgba 0.5703 0 1 1>}
  n21 = PreSkinnedLocalBounds()
  n22 = Subtract(A=VertexInterpolator(VS=PreSkinnedPosition()).PS, B=Add(A=@n21.Min) {constB=0})
  n23 = Add(A=@n21.Extents)
  NamedRerouteDeclaration(None=MaterialFunctionCall(Float3=Divide(A=@n22, B=@n23)) {materialFunction=&BreakOutFloat3Components}.B) {name=meshPosition, nodeColor=<rgba 0 0.0889 1 1>}
  n24 = Multiply(
   A=Subtract(A=Time(), B=ScalarParameter() {defaultValue=-10, group=CrimsonPeak, parameterName=GameTimeFromBP})
   B=Multiply(A=ScalarParameter() {group=CrimsonPeak, parameterName=HealthBar_Speed}) {constB=0.1}
  ) {constB=0}
  NamedRerouteDeclaration(None=OneMinus(None=Frac(None=@n24))) {name=gradientTime, nodeColor=<rgba 0.8984 1 0 1>}
  NamedRerouteUsage()
 ]}
