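# Anim blueprint for the CrimsonPeak backpack cosmetic, fitting the pack to the wearer's spine.
# Each update copies the attach parent's pelvis and spine world transforms onto matching bones, then weighted constraints drive the dyn_spine chain.
# A spine_05-to-spine_03 delta alpha blends in bend corrections; float alphas add female and front-end offsets.
# Gender comes from player pawn 0, and the skydiving float is set but unused in the AnimGraph.
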AnimBlueprint /BRCosmetics/Accessories/FORT_Backpacks/Backpack_M_MED_CrimsonPeak/Meshes/M_MED_CrimsonPeak_Pack_AnimBP
target_skeleton=&/M_MED_CrimsonPeak_Pack_Skeleton
graphs [
 AnimGraph {nodes [
   Local Space Ref Pose()
   n1 = "Transform (Modify) Bone - Bone: pelvis"(
    ComponentPose=Local To Component(LocalPose=Local Space Ref Pose())
    Alpha=1
    Translation=Get pelvis_01_tran() -> (pelvis_01_tran:Vector)
    Rotation=Get pelvis_01_rot() -> (pelvis_01_rot:Rotator)
    Scale="1,1,1"
   ) {
    rotationMode=BMM_Replace
    rotationSpace=BCS_WorldSpace
    scaleSpace=BCS_BoneSpace
    translationMode=BMM_Replace
    translationSpace=BCS_WorldSpace
    alphaBoolBlend {blendOption=Linear}
    alphaScaleBiasClamp {inRange {max=0, min=-40}, outRange {max=0, min=1}}
    boneToModify {boneName=pelvis}
   }
   n2 = "Transform (Modify) Bone - Bone: spine_01"(
    ComponentPose=@n1
    Alpha=1
    Translation=Get spine_01_tran() -> (spine_01_tran:Vector)
    Rotation=Get spine_01_rot() -> (spine_01_rot:Rotator)
    Scale="1,1,1"
   ) {
    rotationMode=BMM_Replace
    rotationSpace=BCS_WorldSpace
    scaleSpace=BCS_BoneSpace
    translationMode=BMM_Replace
    translationSpace=BCS_WorldSpace
    alphaBoolBlend {blendOption=Linear}
    alphaScaleBiasClamp {inRange {max=0, min=-40}, outRange {max=0, min=1}}
    boneToModify {boneName=spine_01}
   }
   n3 = "Transform (Modify) Bone - Bone: spine_02"(
    ComponentPose=@n2
    Alpha=1
    Translation=Get spine_02_tran() -> (spine_02_tran:Vector)
    Rotation=Get spine_02_rot() -> (spine_02_rot:Rotator)
    Scale="1,1,1"
   ) {
    rotationMode=BMM_Replace
    rotationSpace=BCS_WorldSpace
    scaleSpace=BCS_BoneSpace
    translationMode=BMM_Replace
    translationSpace=BCS_WorldSpace
    alphaBoolBlend {blendOption=Linear}
    alphaScaleBiasClamp {inRange {max=0, min=-40}, outRange {max=0, min=1}}
    boneToModify {boneName=spine_02}
   }
   n4 = "Transform (Modify) Bone - Bone: spine_03"(
    ComponentPose=@n3
    Alpha=1
    Translation=Get spine_03_tran() -> (spine_03_tran:Vector)
    Rotation=Get spine_03_rot() -> (spine_03_rot:Rotator)
    Scale="1,1,1"
   ) {
    rotationMode=BMM_Replace
    rotationSpace=BCS_WorldSpace
    scaleSpace=BCS_BoneSpace
    translationMode=BMM_Replace
    translationSpace=BCS_WorldSpace
    alphaBoolBlend {blendOption=Linear}
    alphaScaleBiasClamp {inRange {max=0, min=-40}, outRange {max=0, min=1}}
    boneToModify {boneName=spine_03}
   }
   n5 = "Transform (Modify) Bone - Bone: spine_04"(
    ComponentPose=@n4
    Alpha=1
    Translation=Get spine_04_tran() -> (spine_04_tran:Vector)
    Rotation=Get spine_04_rot() -> (spine_04_rot:Rotator)
    Scale="1,1,1"
   ) {
    rotationMode=BMM_Replace
    rotationSpace=BCS_WorldSpace
    scaleSpace=BCS_BoneSpace
    translationMode=BMM_Replace
    translationSpace=BCS_WorldSpace
    alphaBoolBlend {blendOption=Linear}
    alphaScaleBiasClamp {inRange {max=0, min=-40}, outRange {max=0, min=1}}
    boneToModify {boneName=spine_04}
   }
   n6 = "Transform (Modify) Bone - Bone: spine_05"(
    ComponentPose=@n5
    Alpha=1
    Translation=Get spine_05_tran() -> (spine_05_tran:Vector)
    Rotation=Get spine_05_rot() -> (spine_05_rot:Rotator)
    Scale="1,1,1"
   ) {
    rotationMode=BMM_Replace
    rotationSpace=BCS_WorldSpace
    scaleSpace=BCS_BoneSpace
    translationMode=BMM_Replace
    translationSpace=BCS_WorldSpace
    alphaBoolBlend {blendOption=Linear}
    alphaScaleBiasClamp {inRange {max=0, min=-40}, outRange {max=0, min=1}}
    boneToModify {boneName=spine_05}
   }
   n7 = "Constraint - dyn_spine_4 (P)"(ComponentPose=@n6, Alpha=1, ConstraintWeights_0=1) {
    alphaBoolBlend {blendOption=Linear}
    boneToModify {boneName=dyn_spine_4}
    constraintSetup [{targetBone {boneName=spine_05}}]
    constraintWeights [1]
   }
   n8 = "Constraint - dyn_spine_3 (PP)"!(ComponentPose=@n7, Alpha=1, ConstraintWeights_0=0.1, ConstraintWeights_1:float=0.9) -> (Pose) {
    alphaBoolBlend {blendOption=Linear}
    boneToModify {boneName=dyn_spine_3}
    constraintSetup [{targetBone {boneName=spine_03}}, {targetBone {boneName=spine_05}}]
    constraintWeights [1,1]
   }
   n9 = "Constraint - dyn_spine_2 (PP)"!(ComponentPose=@n8, Alpha=1, ConstraintWeights_0=0.2, ConstraintWeights_1:float=0.8) -> (Pose) {
    alphaBoolBlend {blendOption=Linear}
    boneToModify {boneName=dyn_spine_2}
    constraintSetup [{targetBone {boneName=spine_02}}, {targetBone {boneName=spine_05}}]
    constraintWeights [1,1]
   }
   n10 = "Constraint - dyn_spine_1 (RR)"!(ComponentPose=@n9, Alpha=1, ConstraintWeights_0=0.4, ConstraintWeights_1:float=0.6) -> (Pose) {
    alphaBoolBlend {blendOption=Linear}
    boneToModify {boneName=dyn_spine_1}
    constraintSetup [{transformType=Rotation, targetBone {boneName=pelvis}}, {transformType=Rotation, targetBone {boneName=spine_05}}]
    constraintWeights [1,1]
   }
   n11 = Get spine01_rot() -> (spine01_rot:double)
   n12 = "Constraint - dyn_spine_1 (RR)"(ComponentPose=@n10, Alpha=@n11, ConstraintWeights_0=0.2, ConstraintWeights_1=0.8) {
    alphaBoolBlend {blendOption=Linear}
    boneToModify {boneName=dyn_spine_1}
    constraintSetup [{transformType=Rotation, targetBone {boneName=spine_01}}, {transformType=Rotation, targetBone {boneName=spine_05}}]
    constraintWeights [1,1]
   }
   n13 = "Transform (Modify) Bone - Bone: dyn_spine_2"(ComponentPose=@n12, Alpha=@n11, Translation="0,0,-2", Rotation="0,0,25", Scale="1,1,1") {
    rotationMode=BMM_Additive
    rotationSpace=BCS_BoneSpace
    translationMode=BMM_Additive
    translationSpace=BCS_BoneSpace
    alphaBoolBlend {blendOption=Linear}
    boneToModify {boneName=dyn_spine_2}
    rotation {roll=25}
    translation {z=-2}
   }
   n14 = "Transform (Modify) Bone - Bone: dyn_spine_1"(ComponentPose=@n13, Alpha=@n11, Translation="0,0,-7", Rotation="0,0,20", Scale="1,1,1") {
    rotationMode=BMM_Additive
    rotationSpace=BCS_BoneSpace
    translationMode=BMM_Additive
    translationSpace=BCS_BoneSpace
    alphaBoolBlend {blendOption=Linear}
    boneToModify {boneName=dyn_spine_1}
    rotation {roll=20}
    translation {z=-7}
   }
   n15 = "Constraint - dyn_spine_4 (P)"(ComponentPose=@n14, Alpha=1, ConstraintWeights_0=1) {
    alphaBoolBlend {blendOption=Linear}
    boneToModify {boneName=dyn_spine_4}
    constraintSetup [{targetBone {boneName=pack}}]
    constraintWeights [1]
   }
   n16 = Get isFemaleFloat() -> (isFemaleFloat:double)
   n17 = "Transform (Modify) Bone - Bone: dyn_spine_4"(ComponentPose=@n15, Alpha=@n16, Translation="0,0.5,1.5", Rotation="0,0,0", Scale="0.85,0.85,0.85") {
    scaleMode=BMM_Additive
    scaleSpace=BCS_BoneSpace
    translationMode=BMM_Additive
    translationSpace=BCS_ParentBoneSpace
    alphaBoolBlend {blendOption=Linear}
    boneToModify {boneName=dyn_spine_4}
    scale=<0.85 0.85 0.85>
    translation {y=0.5, z=1.5}
   }
   n18 = "Transform (Modify) Bone - Bone: dyn_spine_3"(ComponentPose=@n17, Alpha=@n16, Translation="0,0,0", Rotation="0,0,0", Scale="1,1,1") {translationMode=BMM_Additive, translationSpace=BCS_ParentBoneSpace, alphaBoolBlend {blendOption=Linear}, boneToModify {boneName=dyn_spine_3}}
   n19 = "Transform (Modify) Bone - Bone: dyn_spine_2"(ComponentPose=@n18, Alpha=@n16, Translation="0,1,0", Rotation="0,0,0", Scale="1,1,1") {
    translationMode=BMM_Additive
    translationSpace=BCS_ParentBoneSpace
    alphaBoolBlend {blendOption=Linear}
    boneToModify {boneName=dyn_spine_2}
    translation {y=1}
   }
   n20 = "Transform (Modify) Bone - Bone: dyn_spine_1"(ComponentPose=@n19, Alpha=@n16, Translation="0,1,0", Rotation="0,0,0", Scale="1,1,1") {
    translationMode=BMM_Additive
    translationSpace=BCS_ParentBoneSpace
    alphaBoolBlend {blendOption=Linear}
    boneToModify {boneName=dyn_spine_1}
    translation {y=1}
   }
   n21 = "Transform (Modify) Bone - Bone: dyn_spine_4"(ComponentPose=@n20, Alpha=Get isFrontEndFloat() -> (isFrontEndFloat:double), Translation="0,2,0", Rotation="0,0,7", Scale="1,1,1") {
    rotationMode=BMM_Additive
    translationMode=BMM_Additive
    translationSpace=BCS_ParentBoneSpace
    alphaBoolBlend {blendOption=Linear}
    boneToModify {boneName=dyn_spine_4}
    rotation {roll=7}
    translation {y=2}
   }
   Output Pose(Result=Component To Local(ComponentPose=@n21)) {layerGroup=None}
  ]}
 EventGraph {nodes [
   n1 = Cast To PlayerPawn_Athena(Object=TryGetPawnOwner())
   > n2 = Set Pawn Owner(Pawn Owner:Fort Player Pawn Athena Object Reference=@n1) -> (Output_Get:Fort Player Pawn Athena Object Reference)
   > Set Skel_Mesh(Skel_Mesh=GetSkeletalMeshForPartType(self=@n2))
   Event BlueprintUpdateAnimation()
   n3 = GetPartSkeletalMeshComponent()
   > Is Valid(InputObject=@n3)
   n4 = GetAttachParent!(self:Scene Component Object Reference=@n3) -> (ReturnValue:Scene Component Object Reference)
   > Is Valid(InputObject=@n4)
   > n5 = Sequence!(execute) -> (then_0, then_1, then_2:Exec, then_3:Exec, then_4:Exec, then_5:Exec, then_6:Exec, then_7:Exec, then_8:Exec, then_9:Exec)
   n6 = BreakTransform(InTransform=GetSocketTransform(self=@n4, InSocketName=spine_01, TransformSpace=RTS_World))
   > Set spine_01_tran(spine_01_tran:Vector=@n6) -> (Output_Get:Vector)
   > Set spine_01_rot(spine_01_rot:Rotator=@n6.Rotation) -> (Output_Get:Rotator)
   n7 = BreakTransform(InTransform=GetSocketTransform(self=@n4, InSocketName=spine_02, TransformSpace=RTS_World))
   Set spine_02_tran(execute=@n5.then_1, spine_02_tran:Vector=@n7) -> (Output_Get:Vector)
   > Set spine_02_rot(spine_02_rot:Rotator=@n7.Rotation) -> (Output_Get:Rotator)
   n8 = BreakTransform(InTransform=GetSocketTransform(self=@n4, InSocketName=spine_03, TransformSpace=RTS_World))
   Set spine_03_tran(execute=@n5.then_2, spine_03_tran:Vector=@n8) -> (Output_Get:Vector)
   > Set spine_03_rot(spine_03_rot:Rotator=@n8.Rotation) -> (Output_Get:Rotator)
   n9 = BreakTransform(InTransform=GetSocketTransform(self=@n4, InSocketName=spine_04, TransformSpace=RTS_World))
   Set spine_04_tran(execute=@n5.then_3, spine_04_tran:Vector=@n9) -> (Output_Get:Vector)
   > Set spine_04_rot(spine_04_rot:Rotator=@n9.Rotation) -> (Output_Get:Rotator)
   n10 = BreakTransform(InTransform=GetSocketTransform(self=@n4, InSocketName=spine_05, TransformSpace=RTS_World))
   Set spine_05_tran(execute=@n5.then_4, spine_05_tran:Vector=@n10) -> (Output_Get:Vector)
   > Set spine_05_rot(spine_05_rot:Rotator=@n10.Rotation) -> (Output_Get:Rotator)
   n11 = BreakTransform(InTransform=GetSocketTransform(self=@n4, InSocketName=pelvis, TransformSpace=RTS_World))
   Set pelvis_01_tran(execute=@n5.then_5, pelvis_01_tran:Vector=@n11) -> (Output_Get:Vector)
   > Set pelvis_01_rot(pelvis_01_rot:Rotator=@n11.Rotation) -> (Output_Get:Rotator)
   n12 = Branch(execute=@n5.then_6, Condition=Get bIsSkyDiving())
   > Set isSkyDivingFloat(isSkyDivingFloat:double=1) -> (Output_Get:double)
   Set isSkyDivingFloat(execute=@n12.else, isSkyDivingFloat=0)
   n13 = FindGender!(execute=@n5.then_7, Object:Object Reference=GetPlayerPawn!(PlayerIndex:int=0) -> (ReturnValue:Pawn Object Reference)) -> (True:Exec, False:Exec, OutGender:EFortCustomGender Enum=Invalid)
   > n14 = Set isFemale(isFemale:bool="Equal (Enum)"(A:EFortCustomGender Enum=@n13, B:EFortCustomGender Enum=Female)) -> (Output_Get:bool)
   > n15 = Branch(Condition=@n14)
   > Set isFemaleFloat(isFemaleFloat:double=1) -> (Output_Get:double)
   Set isFemaleFloat(execute=@n15.else, isFemaleFloat=0)
   n16 = Branch(execute=@n5.then_8, Condition=Get bIsInFrontEnd())
   > Set isFrontEndFloat(isFrontEndFloat:double=1) -> (Output_Get:double)
   Set isFrontEndFloat(execute=@n16.else, isFrontEndFloat=0)
   Set spine01_rot(
    execute=@n5.then_9
    spine01_rot:double=GetAlphaFromReferenceTransformDelta(SkeletalMeshComponent=Get Skel_Mesh(), BoneName=spine_05, BaseName=spine_03, InputRange=(X=0,Y=-20), TwistAxis="0,0,1")
   ) -> (Output_Get:double)
   Event BlueprintInitializeAnimation()
   > Set Skel_Mesh(Skel_Mesh=GetOwnerSkeletalMeshComponent())
   TryGetPawnOwner()
   TryGetPawnOwner()
  ]}
]
variables [
 Gender {type=EFortCustomGender, default=Invalid}
 isFemale {type=boolean}
 isFemaleAndFrontEndFloat {type=number}
 isFemaleFloat {type=number}
 isFemaleNewMethodFloat {type=number}
 isFrontEndFloat {type=number}
 isSkyDivingFloat {type=number}
 Pawn Owner {type=FortPlayerPawnAthena}
 pelvis_01_rot {type=Rotator}
 pelvis_01_tran {type=Vector}
 Skel_Mesh {type=SkeletalMeshComponent}
 spine01_rot {type=number}
 spine_01_rot {type=Rotator}
 spine_01_tran {type=Vector}
 spine_02_rot {type=Rotator}
 spine_02_tran {type=Vector}
 spine_03_rot {type=Rotator}
 spine_03_tran {type=Vector}
 spine_04_rot {type=Rotator}
 spine_04_tran {type=Vector}
 spine_05_rot {type=Rotator}
 spine_05_tran {type=Vector}
]
